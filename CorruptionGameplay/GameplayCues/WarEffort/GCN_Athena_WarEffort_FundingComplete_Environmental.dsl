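; Gameplay cue that tunes environmental burst VFX when a War Effort donation station completes funding.
; On burst, the rounded normalized magnitude encodes station type (Vault/Weapons, Guard Station, Brute Pad), while rounded raw magnitude picks box offset and Brute Pad box size.
; Each Niagara component receives the cue location and index-selected user parameters for box, sprite scale, smoke, spawning and camera offset.
; The station-type switch has empty branches.
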(fn OnBurst(EitherVFX-preferthisversion)
  (params (MyTarget :type Actor%20Object%20Reference) (Parameters :type Gameplay%20Cue%20Parameters%20Structure%20%28by%20ref%29) (ParticleComponents :type Array%20of%20FXSystem%20Component%20Object%20References) (AudioComponents :type Array%20of%20Audio%20Component%20Object%20References) (BurstCameraShakeInstance :type Camera%20Shake%20Base%20Object%20Reference) (BurstDecalInstance :type Decal%20Actor%20Object%20Reference))
  (bind _parameters (Variables|GetParameters))
  (bind (_normalizedmagnitude _rawmagnitude _effectcontext _matchedtagname _originaltag _aggregatedsourcetags _aggregatedtargettags _location) (Ability|GameplayCue|BreakGameplayCueParameters _parameters))
  (bind _returnvalue (Math|Float|Round _rawmagnitude))
  (bind _returnvalue_2 (Math|Float|Round _normalizedmagnitude))
  (Utilities|FlowControl|Sequence
    (:then_0
      ;[1 1 1] basic ID for donation station type :
      ; 0 - Vault/Weapons
      ; 1 - Guard Station
      ; 2 - Brute Pad
      (switch int _returnvalue_2
        (:1
          _)
        (:2
          _))
      (for _array_element ParticleComponents
        (bind _asniagara_particle_system_component (Utilities|Casting|CastToNiagaraComponent _array_element))
        (Niagara|SetNiagaraVariable(Vector3) _asniagara_particle_system_component "User.SourceOrigin" _location)
        (Niagara|SetNiagaraVariable(Vector3) _asniagara_particle_system_component "User.BoxSourceSize" (Utilities|Select :Index _returnvalue_2 :Option%200 "0, 0, 0" :Option%201 "220.000000,220.000000,20.000000" :Option%202 (Utilities|Select :Index _returnvalue :Option%200 "400.000000,400.000000,5.000000" :Option%201 "600.000000,300.000000,5.000000")))
        (Niagara|SetNiagaraVariable(Vector3) _asniagara_particle_system_component "User.BoxOffset" (Utilities|Select :Index _returnvalue :Option%200 "0, 0, 0" :Option%201 "-100.000000,-50.000000,0.000000"))
        (Niagara|SetNiagaraVariable(Float) _asniagara_particle_system_component "User.SpriteScale" (Utilities|Select :Index _returnvalue_2 :Option%200 0.0 :Option%201 1.0 :Option%202 1.5))
        (Niagara|SetNiagaraVariable(Float) _asniagara_particle_system_component "User.SmokeDuration" (Utilities|Select :Index _returnvalue_2 :Option%200 0.0 :Option%201 3.0 :Option%202 4.0))
        (Niagara|SetNiagaraVariable(Int32) _asniagara_particle_system_component "User.SpawnCount" (Utilities|Select :Index _returnvalue_2 :Option%200 0 :Option%201 20 :Option%202 20))
        (Niagara|SetNiagaraVariable(Float) _asniagara_particle_system_component "User.CamOffset" (Utilities|Select :Index _returnvalue_2 :Option%200 0.0 :Option%201 25.0 :Option%202 1.5e+02))
        (Niagara|SetNiagaraVariable(Float) _asniagara_particle_system_component "User.SpawnDelay" (Utilities|Select :Index _returnvalue_2 :Option%200 0.0 :Option%201 0.0 :Option%202 1.0))))
    (:then_1
      (return))))
